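# Unlit prototype Landmass material that generates a radial cup-shaped landscape height, packed by PackRG8Height.
# Distance from UV centre minus radius R passes through a smooth max against zero, then is scaled by HH and clamped before packing.
# Separate branches, not feeding the cup, hold ramp/plateau/channel/lip shaping, height pack/unpack debugging and loose parameters.
# Another builds multi-octave gradient noise in a Custom HLSL loop, tiled using MPC_Erosion resolution values.
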Material /Landmass/Landscape/OldPrototype_BP/Mats/Seed/GenNoise_02_Cup
graph {
 properties {bAllowNegativeEmissiveColor=true, pixelDepthOffsetMode=PDOM_Legacy, shadingModel=MSM_Unlit, usedShadingModels=Unlit}
 nodes [
  n1 = Subtract(A=TextureCoordinate()) {constB=0.5}
  n2 = Subtract(
   A=MaterialFunctionCall(Vector 3=@n1, Vector 2=@n1) {materialFunction=/Engine/Functions/Engine_MaterialFunctions02/Utility/VectorLength}.V2 Length
   B=ScalarParameter() {defaultValue=0.2, parameterName=R}
  )
  n3 = Constant()
  n4 = ScalarParameter() {defaultValue=0.035, parameterName=asdf}
  n5 = Clamp(None=Multiply(
    A=MaterialFunctionCall!(A=@n2, B=@n3, K=@n4) {materialFunction=/Landmass/Landscape/OldPrototype_BP/Mats/MF/SmoothIntersections}.Smooth Max
    B=ScalarParameter() {defaultValue=354000, parameterName=HH}
   )) {maxDefault=20000}
  Add(A=@n5)
  n6 = ComponentMask(None=TextureCoordinate()) {r=true}
  n7 = Multiply(A=Abs(None=Subtract(A=@n6) {constB=0.5})) {constB=2}
  n8 = ScalarParameter() {defaultValue=0.9, parameterName=LipWidth}
  n9 = VectorParameter() {desc="R=Ramp, G=Plateau, B=Channel, A=Lip", parameterName=TestFeatureParams}
  n10 = TextureCoordinate()
  n11 = Abs(None=Subtract(A=ComponentMask(None=@n10) {g=true}) {constB=0.5})
  n12 = ComponentMask(None=@n10) {r=true}
  n13 = Subtract(A=@n12) {constB=0.5}
  n14 = Multiply(A=Saturate(None=Ceil(None=@n13)), B=@n9.G)
  n15 = Multiply(A=@n12, B=@n9.R)
  Add(
   A=Multiply(
    A=MaterialFunctionCall!(Alpha=@n7, Offset=@n8) {materialFunction=/Engine/Functions/Engine_MaterialFunctions01/Opacity/AlphaOffset}.Result
    B=@n9.A
   )
   B=Add(A=Multiply(A=@n11, B=@n9.B), B=Add(A=@n14, B=@n15))
  )
  n16 = ScalarParameter() {defaultValue=-32768, parameterName=H}
  n17 = MaterialFunctionCall!(Height Int -32k to 32k=@n16) {materialFunction=/Landmass/Landscape/OldPrototype_BP/Mats/MF/PackRG8Height}
  Divide(A=@n17.XY) {constB=255}
  n18 = VectorParameter() {desc="R=Elevation Max, G=Elevation Min, B=Pre-Offset, A=Invert", parameterName=ElevationParams}
  n19 = TextureCoordinate()
  n20 = Subtract(A=AppendVector(
    A=CollectionParameter() {collection=/Landmass/Landscape/BlueprintBrushes/MPC/MPC_Erosion, parameterName=ResX}
    B=CollectionParameter() {collection=&MPC_Erosion, parameterName=ResY}
   ))
  n21 = Subtract(A=CollectionParameter() {collection=&MPC_Erosion, parameterName=ComponentRes})
  n22 = Max(A=Floor(None=Divide(A=@n20, B=@n21) {constB=256}))
  MaterialExpressionVectorParameter_0 = VectorParameter() {desc="R=Tiling, G=Seed, B=Octaves, A=Persistence", parameterName=NoiseSeedParams, defaultValue=<rgba 2 0 2 0.5>}
  MaterialExpressionAppendVector_0 = AppendVector(A=Multiply(A=Multiply(A=@n19, B=@n22), B=@MaterialExpressionVectorParameter_0.R), B=@MaterialExpressionVectorParameter_0.G)
  MaterialExpressionConstant_2 = Constant() {desc="3 = Gradient, 5 = Voronoi", r=3}
  MaterialExpressionConstant_3 = Constant()
  MaterialExpressionMultiply_9 = Multiply(A=@n22, B=@MaterialExpressionVectorParameter_0.R)
  n23 = Custom!(
   Position=@MaterialExpressionAppendVector_0
   Function=@MaterialExpressionConstant_2
   bTurbulence=@MaterialExpressionConstant_3
   Octaves=@MaterialExpressionVectorParameter_0.B
   Persistence=@MaterialExpressionVectorParameter_0.B
   RepeatSize=@MaterialExpressionMultiply_9
  ) {
   code="\r\nfloat Out = 0;\r\nfloat OutScale = 1;\r\n\r\n[loop]\r\nfor(int i = 0; i < Octaves; ++i)\r\n{\r\n\r\n\r\n\tif(bTurbulence)\r\n\t{\r\n\t\tOut += abs(Noise3D_Multiplexer(Function, Position, 4, 1, RepeatSize)) * OutScale;\r\n\t}\r\n\telse\r\n\t{\r\n\t\tOut += Noise3D_Multiplexer(Function, Position, 4, 1, RepeatSize) * OutScale;\r\n\t}\r\n\r\n\tPosition *= 2;\r\n\tRepeatSize *= 2;\r\n\tOutScale *= Persistence;\r\n\r\n}\r\n\r\nif(!bTurbulence)\r\n{\r\n\t// bring -1..1 to 0..1 range\r\n\tOut = Out * 0.5f + 0.5f;\r\n}\r\n\r\n// Out is in 0..1 range\r\nreturn lerp(0, 1, Out);"
   description=CustomNoise
   outputType=CMOT_Float1
   inputs [
    {inputName=Position, input {expression=&self:MaterialExpressionAppendVector_0}}
    {inputName=Function, input {expression=&self:MaterialExpressionConstant_2}}
    {inputName=bTurbulence, input {expression=&self:MaterialExpressionConstant_3}}
    {
     inputName=Octaves
     input {expression=&self:MaterialExpressionVectorParameter_0, mask=1, maskB=1, outputIndex=3}
    }
    {
     inputName=Persistence
     input {expression=&self:MaterialExpressionVectorParameter_0, mask=1, maskA=1, outputIndex=4}
    }
    {inputName=RepeatSize, input {expression=&self:MaterialExpressionMultiply_9}}
   ]
  }
  n24 = LinearInterpolate(A=@n23, B=OneMinus(None=@n23), Alpha=@n18.A)
  LinearInterpolate(A=@n18.G, B=@n18.G, Alpha=Max(A=Subtract(A=@n24, B=@n18.B)) {constB=0})
  n25 = Floor(None=Add(A=@n16) {constB=32768, desc="Make  0-65536"})
  n26 = Floor(None=Divide(A=@n25) {constB=256})
  n27 = Fmod(A=@n25, B=Constant() {r=256})
  n28 = Divide(A=Floor(None=AppendVector(A=@n26, B=@n27))) {constB=255}
  MaterialFunctionCall!(Vector2=@n28, MaximumNumberOfDigits=Constant() {r=8}, UVs, DebugTextLocation RG_UpperRight BA_LowerLeft, Component Spacing) {materialFunction=/Engine/Functions/Engine_MaterialFunctions02/Utility/DebugFloat2Values}
  MaterialFunctionCall(Height Int -32k to 32k=Clamp(None=@n5) {maxDefault=32767, minDefault=-32768}) {materialFunction=&PackRG8Height}
  Max(A=@n2) {constB=0}
  Multiply(A=@n17.XY) {constB=255}
  Noise!(World Position=@MaterialExpressionAppendVector_0, FilterWidth) {bTurbulence=false, levelScale=3, levels=16, noiseFunction=NOISEFUNCTION_VoronoiALU, outputMin=0, quality=5, repeatSize=2}
  ScalarParameter() {parameterName=Seed}
  ScalarParameter() {defaultValue=2, parameterName=Tiling}
  ScalarParameter() {parameterName=Channel}
  ScalarParameter() {defaultValue=5, parameterName=NoiseType}
  ScalarParameter() {defaultValue=0.5, parameterName=Persistence}
  ScalarParameter() {parameterName=LipHeight}
  ScalarParameter() {defaultValue=0.3, parameterName=Elevation Pre-Offset}
  ScalarParameter() {defaultValue=3, parameterName=Octaves}
  ScalarParameter() {defaultValue=1, parameterName=Invert}
  ScalarParameter() {parameterName=Ramp}
  ScalarParameter() {parameterName=Plateau}
  Sign(None=MaterialFunctionCall!(RG=@n28) {materialFunction=/Landmass/Landscape/OldPrototype_BP/Mats/MF/UnpackRG8Height}.Result)
 ]
}
