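# Unlit prototype Landmass seed material that generates Voronoi-noise terrain height for a landscape brush, apparently for erosion given its MPC_Erosion inputs.
# A custom HLSL loop sums Noise3D_Multiplexer octaves over resolution-scaled, offset UVs, with seed as the third coordinate.
# The invertible noise drives an elevation lerp, added to UV-based ramp, plateau, channel and lip shapes.
# The height is clamped to signed 16-bit range and packed by PackRG8Height.
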Material /Landmass/Landscape/OldPrototype_BP/Mats/Seed/GenNoise_02_Voronoi
graph {
 properties {bAllowNegativeEmissiveColor=true, pixelDepthOffsetMode=PDOM_Legacy, shadingModel=MSM_Unlit, usedShadingModels=Unlit}
 nodes [
  n1 = Constant3Vector() {constant=<rgba 0 0 1 0>}
  n2 = ScalarParameter() {parameterName=ang}
  n3 = Constant() {r=0.5}
  n4 = AppendVector(A=TextureCoordinate(), B=Constant())
  n5 = TextureCoordinate()
  n6 = Divide(A=AppendVector(
    A=CollectionParameter() {collection=/Landmass/Landscape/BlueprintBrushes/MPC/MPC_Erosion, parameterName=ResX}
    B=CollectionParameter() {collection=&MPC_Erosion, parameterName=ResY}
   )) {constB=256}
  n7 = Subtract(A=Multiply(A=@n5, B=@n6), B=Add(A=ScalarParameter() {parameterName=ManualOffset}, B=Divide(A=@n6)))
  Add(
   A=ComponentMask(None=RotateAboutAxis(NormalizedRotationAxis=@n1, RotationAngle=@n2, PivotPoint=@n3, Position=@n4)) {g=true, r=true}
   B=Frac(None=@n7)
  )
  n8 = Subtract(A=ComponentMask(None=TextureCoordinate()) {r=true}) {constB=0.5}
  n9 = MaterialFunctionCall!(Alpha=Multiply(A=Abs(None=@n8)) {constB=2}, Offset=ScalarParameter() {defaultValue=0.9, parameterName=LipWidth}) {materialFunction=/Engine/Functions/Engine_MaterialFunctions01/Opacity/AlphaOffset}
  n10 = VectorParameter() {desc="R=Ramp, G=Plateau, B=Channel, A=Lip", parameterName=TestFeatureParams}
  n11 = ComponentMask(None=@n5) {g=true}
  n12 = Multiply(A=Abs(None=Subtract(A=@n11) {constB=0.5}), B=@n10.B)
  n13 = ComponentMask(None=@n5) {r=true}
  n14 = Ceil(None=Subtract(A=@n13) {constB=0.5})
  n15 = Add(A=Multiply(A=Saturate(None=@n14), B=@n10.G), B=Multiply(A=@n13, B=@n10.R))
  n16 = VectorParameter() {desc="R=Elevation Max, G=Elevation Min, B=Pre-Offset, A=Invert", parameterName=ElevationParams}
  MaterialExpressionVectorParameter_0 = VectorParameter() {desc="R=Tiling, G=Seed, B=Octaves, A=Persistence", parameterName=NoiseSeedParams, defaultValue=<rgba 2 0 2 0.5>}
  MaterialExpressionAppendVector_0 = AppendVector(A=Multiply(A=@n7, B=@MaterialExpressionVectorParameter_0.R), B=@MaterialExpressionVectorParameter_0.G)
  MaterialExpressionConstant_2 = Constant() {desc="3 = Gradient, 5 = Voronoi", r=5}
  MaterialExpressionConstant_3 = Constant()
  n17 = Custom!(
   Position=@MaterialExpressionAppendVector_0
   Function=@MaterialExpressionConstant_2
   bTurbulence=@MaterialExpressionConstant_3
   Octaves=@MaterialExpressionVectorParameter_0.B
   Persistence=@MaterialExpressionVectorParameter_0.B
   RepeatSize=@MaterialExpressionVectorParameter_0.B
  ) {
   code="\r\nfloat Out = 0;\r\nfloat OutScale = 1;\r\n\r\n[loop]\r\nfor(int i = 0; i < Octaves; ++i)\r\n{\r\n\r\n\r\n\tif(bTurbulence)\r\n\t{\r\n\t\tOut += abs(Noise3D_Multiplexer(Function, Position, 4, 0, RepeatSize)) * OutScale;\r\n\t}\r\n\telse\r\n\t{\r\n\t\tOut += Noise3D_Multiplexer(Function, Position, 4, 0, RepeatSize) * OutScale;\r\n\t}\r\n\r\n\tPosition *= 2;\r\n\tRepeatSize *= 2;\r\n\tOutScale *= Persistence;\r\n\r\n}\r\n\r\nif(!bTurbulence)\r\n{\r\n\t// bring -1..1 to 0..1 range\r\n\tOut = Out * 0.5f + 0.5f;\r\n}\r\n\r\n// Out is in 0..1 range\r\nreturn lerp(0, 1, Out);"
   description=CustomNoise
   outputType=CMOT_Float1
   inputs [
    {inputName=Position, input {expression=&self:MaterialExpressionAppendVector_0}}
    {inputName=Function, input {expression=&self:MaterialExpressionConstant_2}}
    {inputName=bTurbulence, input {expression=&self:MaterialExpressionConstant_3}}
    {
     inputName=Octaves
     input {
      expression=&self:MaterialExpressionVectorParameter_0
      mask=1
      maskB=1
      outputIndex=3
     }
    }
    {
     inputName=Persistence
     input {
      expression=&self:MaterialExpressionVectorParameter_0
      mask=1
      maskA=1
      outputIndex=4
     }
    }
    {
     inputName=RepeatSize
     input {
      expression=&self:MaterialExpressionVectorParameter_0
      mask=1
      maskR=1
      outputIndex=1
     }
    }
   ]
  }
  n18 = Subtract(A=LinearInterpolate(A=@n17, B=OneMinus(None=@n17), Alpha=@n16.A), B=@n16.B)
  n19 = Add(A=Add(A=Multiply(A=@n9.Result, B=@n10.A), B=Add(A=@n12, B=@n15)), B=LinearInterpolate(A=@n16.G, B=@n16.G, Alpha=@n18))
  n20 = TextureSampleParameter2D() {parameterName=SeedRT, texture=&Black}
  n21 = Multiply(A=Multiply(A=@n20.R) {constB=255}) {constB=256}
  n22 = Multiply(A=@n20.G) {constB=255}
  Add(A=@n19, B=Subtract(A=Add(A=@n21, B=@n22)) {constB=32768})
  MaterialFunctionCall(X=@n19, Y=Constant(), Z=CollectionParameter() {collection=&MPC_Erosion, parameterName=StartWater}) {materialFunction=&MakeFloat3}
  MaterialFunctionCall!(Height Int -32k to 32k=Clamp(None=@n19) {maxDefault=32767, minDefault=-32768}) {materialFunction=/Landmass/Landscape/OldPrototype_BP/Mats/MF/PackRG8Height}
  Max(A=@n18) {constB=0}
  Noise!(World Position=@MaterialExpressionAppendVector_0, FilterWidth) {bTurbulence=false, levelScale=3, levels=16, noiseFunction=NOISEFUNCTION_VoronoiALU, outputMin=0, quality=5, repeatSize=2}
  ScalarParameter() {parameterName=Seed}
  ScalarParameter() {defaultValue=2, parameterName=Tiling}
  ScalarParameter() {parameterName=Channel}
  ScalarParameter() {defaultValue=5, parameterName=NoiseType}
  ScalarParameter() {defaultValue=0.5, parameterName=Persistence}
  ScalarParameter() {parameterName=LipHeight}
  ScalarParameter() {defaultValue=0.3, parameterName=Elevation Pre-Offset}
  ScalarParameter() {defaultValue=3, parameterName=Octaves}
  ScalarParameter() {parameterName=Elevation Min}
  ScalarParameter() {defaultValue=32000, parameterName=Elevation Max}
  ScalarParameter() {defaultValue=1, parameterName=Invert}
  ScalarParameter() {parameterName=Ramp}
  ScalarParameter() {parameterName=Plateau}
 ]
}
