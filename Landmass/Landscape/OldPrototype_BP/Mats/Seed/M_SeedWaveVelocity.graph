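# Prototype unlit Landmass material for seeding wave velocity, with negative emissive values allowed; the graph shows no connection to a material output.
# Water depth comes from MPC_Erosion StartWater minus LandscapeRT height plus an offset, clamped at zero.
# A custom tanh node feeds the depth-dependent dispersion relation, whose frequency scales direction V and wavelength W, summed with a trochoidal velocity term.
# Several branches are unused leftovers, including unpacked RG8 heights, world position and a deep-water frequency.
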Material /Landmass/Landscape/OldPrototype_BP/Mats/Seed/M_SeedWaveVelocity
graph {
 properties {bAllowNegativeEmissiveColor=true, pixelDepthOffsetMode=PDOM_Legacy, shadingModel=MSM_Unlit, usedShadingModels=Unlit}
 nodes [
  n1 = TextureSampleParameter2D() {parameterName=WaterRT, texture=&Black}
  Add(
   A=MaterialFunctionCall!(RG=TextureSampleParameter2D() {parameterName=LSRT, texture=&Black}.RGB) {materialFunction=/Landmass/Landscape/OldPrototype_BP/Mats/MF/UnpackRG8Height}.Result
   B=MaterialFunctionCall(RG=@n1.RGB) {materialFunction=&UnpackRG8Height}.Result
  )
  n2 = ComponentMask() {b=true}
  n3 = VectorParameter() {parameterName=V, defaultValue=<rgba 0 1 0 0>}
  n4 = Multiply(A=ScalarParameter() {defaultValue=8000, parameterName=W}, B=Normalize(VectorInput=ComponentMask(None=@n3.RGB) {g=true, r=true}))
  n5 = Divide(
   A=MaterialFunctionCall(Multiplier=Constant() {r=2}) {materialFunction=&Pi}.Result
   B=ScalarParameter() {defaultValue=8000, parameterName=W}
  ) {desc=k}
  n6 = ScalarParameter() {defaultValue=980, group=Technical, parameterName=Gravitational Constant}
  n7 = TextureSampleParameter2D() {parameterName=LandscapeRT, texture=&Black}
  n8 = Subtract(A=CollectionParameter() {collection=/Landmass/Landscape/BlueprintBrushes/MPC/MPC_Erosion, parameterName=StartWater}, B=@n7.R)
  n9 = ScalarParameter() {parameterName=Add Wave Height}
  MaterialExpressionMultiply_23 = Multiply(A=@n5, B=Max(A=Add(A=@n8, B=@n9)) {constB=0, desc=D})
  n10 = Custom!(x=@MaterialExpressionMultiply_23) {
   code="return tanh(x);"
   description=Tanh
   outputType=CMOT_Float1
   inputs [{inputName=x, input {expression=&self:MaterialExpressionMultiply_23}}]
  }
  n11 = SquareRoot(None=Multiply(A=Multiply(A=@n5, B=@n6), B=@n10)) {desc=w}
  Add(
   A=Multiply(A=Multiply(A=@n2, B=@n4), B=ScalarParameter() {parameterName=Trochoidal Wave Velocity}) {constB=0.125}
   B=Multiply(A=Multiply(A=@n4, B=@n11))
  )
  n12 = TextureCoordinate()
  AppendVector(A=Multiply(A=Subtract(A=@n12) {constB=0.5}, B=ScalarParameter() {defaultValue=102000, parameterName=World Size}), B=Constant())
  Max(A=@n7.B) {constB=0}
  Multiply(A=OneMinus(None=@n1.A)) {constB=10000}
  Normalize(VectorInput=ComponentMask(None=VectorParameter() {parameterName=V, defaultValue=<rgba 0 1 0 0>}.RGB) {g=true, r=true})
  ScalarParameter() {defaultValue=8000, parameterName=W}
  ScalarParameter() {defaultValue=1, parameterName=A}
  ScalarParameter() {defaultValue=1, parameterName=vx}
  ScalarParameter() {defaultValue=1, parameterName=S}
  ScalarParameter() {parameterName=Time}
  n13 = MaterialFunctionCall(Multiplier=Constant() {r=2}) {materialFunction=&Pi}
  n14 = ScalarParameter() {defaultValue=8000, parameterName=W}
  SquareRoot(None=Multiply(A=Divide(A=@n13.Result, B=@n14) {desc=k}, B=ScalarParameter() {defaultValue=980, group=Technical, parameterName=Gravitational Constant})) {desc=w}
 ]
}
