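# Unlit seed material in an old Landmass landscape prototype that builds procedural gradient-noise height values.
# Texture coordinates gain a Z parameter as third component, are scaled by T and sample computed gradient noise.
# The noise is scaled by MaxZ and fed to PackRG8Height, which takes signed 16-bit-range height.
# A noise-texture divide branch sits unused, and no material output connection appears.
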Material /Landmass/Landscape/OldPrototype_BP/Mats/Seed/SimpleNoise
graph {
 properties {pixelDepthOffsetMode=PDOM_Legacy, shadingModel=MSM_Unlit, usedShadingModels=Unlit}
 nodes [
  Divide(A=TextureSample() {samplerType=SAMPLERTYPE_LinearColor, texture=/Landmass/TextureCreation/Textures/T_Noise_PNG_1024}.RGB) {constB=5000}
  n1 = Multiply(A=AppendVector(A=TextureCoordinate(), B=ScalarParameter() {parameterName=Z}), B=ScalarParameter() {defaultValue=1, parameterName=T})
  MaterialFunctionCall!(Height Int -32k to 32k=Multiply(
    A=Noise!(World Position=@n1, FilterWidth) {noiseFunction=NOISEFUNCTION_GradientALU, outputMin=0}
    B=ScalarParameter() {defaultValue=0.5, parameterName=MaxZ}
   )) {materialFunction=/Landmass/Landscape/OldPrototype_BP/Mats/MF/PackRG8Height}
 ]
}
