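# Unlit seed material for the prototype Landmass landscape erosion, building starting terrain height and water values from render targets.
# It unpacks RG8 heights from SeedRT and LandscapeRT, sums them into X of a float3.
# Z blends by UseWaterLevel between uniform StartWater and StartWater minus height clamped at zero, for level water, both parameters from MPC_Erosion.
# Separate LSRT plus WaterRT height sum and scaled inverted WaterRT alpha branches feed no other node.
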Material /Landmass/Landscape/OldPrototype_BP/Mats/Seed/UnpackRG8Height
graph {
 properties {bAllowNegativeEmissiveColor=true, pixelDepthOffsetMode=PDOM_Legacy, shadingModel=MSM_Unlit, usedShadingModels=Unlit}
 nodes [
  n1 = TextureSampleParameter2D() {parameterName=WaterRT, texture=&Black}
  Add(
   A=MaterialFunctionCall!(RG=TextureSampleParameter2D() {parameterName=LSRT, texture=&Black}.RGB) {materialFunction=/Landmass/Landscape/OldPrototype_BP/Mats/MF/UnpackRG8Height}.Result
   B=MaterialFunctionCall(RG=@n1.RGB) {materialFunction=&UnpackRG8Height}.Result
  )
  n2 = Add(
   A=MaterialFunctionCall(RG=TextureSampleParameter2D() {parameterName=SeedRT, texture=&Black}.RGB) {materialFunction=&UnpackRG8Height}.Result
   B=MaterialFunctionCall(RG=TextureSampleParameter2D() {parameterName=LandscapeRT, texture=&Black}.RGB) {materialFunction=&UnpackRG8Height}.Result
  )
  n3 = CollectionParameter() {collection=/Landmass/Landscape/BlueprintBrushes/MPC/MPC_Erosion, parameterName=StartWater}
  n4 = Subtract(A=@n3, B=@n2)
  MaterialFunctionCall(
   X=@n2
   Z=LinearInterpolate(
    A=@n3
    B=Max(A=@n4) {constB=0, desc="Use for Water to be level,  not on top of terrain"}
    Alpha=CollectionParameter() {collection=&MPC_Erosion, parameterName=UseWaterLevel}
   )
  ) {materialFunction=&MakeFloat3}
  Multiply(A=OneMinus(None=@n1.A)) {constB=10000}
 ]
}
